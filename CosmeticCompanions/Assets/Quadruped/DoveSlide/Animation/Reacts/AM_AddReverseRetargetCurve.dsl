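; AM_AddReverseRetargetCurve is an animation modifier for the DoveSlide quadruped cosmetic companion's react animations.
; When applied, it adds a "ReverseRetarget" float curve with a constant value of 1.0 to the target anim sequence.
; It is a single utility call on apply, with no revert logic shown.
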(event EventOnApply
  (params (AnimationSequence :type Anim%20Sequence%20Object%20Reference))
  (Animation|Utilities|AddCurvetoAnimSequenceBasewithValue AnimationSequence "ReverseRetarget" 1.0 true))
