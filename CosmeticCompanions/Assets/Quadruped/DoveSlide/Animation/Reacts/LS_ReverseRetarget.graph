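# Level Sequence for DoveSlide quadruped companion react animations, named for reverse retargeting, that plays a short animation on two characters together.
# It spawns a FortCompanionMannequin and a DoveSlide_Quadruped skeletal mesh actor, each with an identity transform and a skeletal animation section over the same frame range.
# An Animation Mode track drives the mannequin's child mesh component.
# No Control Rig or retargeting tracks appear in the sequence.
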LevelSequence /CosmeticCompanions/Assets/Quadruped/DoveSlide/Animation/Reacts/LS_ReverseRetarget
bindings [
 FortCompanionMannequin {
  id=#1
  possessed_class=FortCompanionMannequin
  tracks [
   {
    type=MovieScene3DTransformTrack
    sections [{
      type=MovieScene3DTransformSection
      channels [
       Location.X {type=MovieSceneScriptingDoubleChannel, default=0}
       Location.Y {type=MovieSceneScriptingDoubleChannel, default=0}
       Location.Z {type=MovieSceneScriptingDoubleChannel, default=0}
       Rotation.X {type=MovieSceneScriptingDoubleChannel, default=0}
       Rotation.Y {type=MovieSceneScriptingDoubleChannel, default=0}
       Rotation.Z {type=MovieSceneScriptingDoubleChannel, default=0}
       Scale.X {type=MovieSceneScriptingDoubleChannel, default=1}
       Scale.Y {type=MovieSceneScriptingDoubleChannel, default=1}
       Scale.Z {type=MovieSceneScriptingDoubleChannel, default=1}
      ]
     }]
   }
   {type=MovieSceneSkeletalAnimationTrack, sections [{type=MovieSceneSkeletalAnimationSection, end_frame=24, start_frame=0}]}
   {type=MovieSceneSpawnTrack, sections [{type=MovieSceneSpawnSection, channels [None {type=MovieSceneScriptingBoolChannel, default=true}]}]}
  ]
 }
 SkeletalMeshComponent0 {
  id=#2
  parent_id=#1
  possessed_class=SkeletalMeshComponent
  tracks [{
    type=MovieSceneByteTrack
    display_name=Animation Mode
    sections [{type=MovieSceneByteSection, channels [None {type=MovieSceneScriptingByteChannel, default=2, const [2,1]}]}]
   }]
 }
 DoveSlide_Quadruped {
  id=#3
  possessed_class=FortSkeletalMeshActor
  tracks [
   {
    type=MovieScene3DTransformTrack
    sections [{
      type=MovieScene3DTransformSection
      channels [
       Location.X {type=MovieSceneScriptingDoubleChannel, default=0}
       Location.Y {type=MovieSceneScriptingDoubleChannel, default=0}
       Location.Z {type=MovieSceneScriptingDoubleChannel, default=0}
       Rotation.X {type=MovieSceneScriptingDoubleChannel, default=0}
       Rotation.Y {type=MovieSceneScriptingDoubleChannel, default=0}
       Rotation.Z {type=MovieSceneScriptingDoubleChannel, default=0}
       Scale.X {type=MovieSceneScriptingDoubleChannel, default=1}
       Scale.Y {type=MovieSceneScriptingDoubleChannel, default=1}
       Scale.Z {type=MovieSceneScriptingDoubleChannel, default=1}
      ]
     }]
   }
   {type=MovieSceneSkeletalAnimationTrack, sections [{type=MovieSceneSkeletalAnimationSection, end_frame=24, start_frame=0}]}
   {type=MovieSceneSpawnTrack, sections [{type=MovieSceneSpawnSection, channels [None {type=MovieSceneScriptingBoolChannel, default=true}]}]}
  ]
 }
]
display_rate {denominator=1, numerator=30}
tick_resolution {denominator=1, numerator=24000}
tracks []
